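# Physics asset for the Pumpkin Punk backpack cosmetic, giving its bats and branch bones limited secondary motion.
# A small box body on the pack anchors two long capsules for the left and right bats.
# Two short capsule chains, a three-segment branch and a two-segment upper branch, both hang off the right bat.
# Tight swing and twist limits on every joint keep the motion to a subtle wobble.
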PhysicsAsset /BRCosmetics/Accessories/FORT_Backpacks/Backpack_M_MED_Pumpkin_Punk/Meshes/M_MED_Pumpkin_Punk_Pack_Physics
bodies {
 bat_l [{shapeType=Capsule, length=48, radius=3, center=<-7.2058 -0.0341 -0.0367>, rotation=<rot 90 0 0>}]
 bat_r [{shapeType=Capsule, length=48, radius=3, center=<-6.9807 -0.02 0.0663>, rotation=<rot 90 0 0>}]
 branch1 [{shapeType=Capsule, length=2.1927, radius=0.7966, center=<1.9057 0 -0.0001>, rotation=<rot 90 0 0>}]
 branch2 [{shapeType=Capsule, length=1.7653, radius=0.7, center=<1.2867 0 0>, rotation=<rot 90 0 0>}]
 branch3 [{shapeType=Capsule, length=2.6399, radius=0.7404, center=<2.1399 -0.1828 -0.4562>, rotation=<rot 90 0 0>}]
 branchUpr1 [{shapeType=Capsule, length=1.5, radius=0.5, center=<1.3128 0 0>, rotation=<rot 90 0 0>}]
 branchUpr2 [{shapeType=Capsule, length=1.9275, radius=0.5, center=<1.4092 0 0>, rotation=<rot 90 0 0>}]
 pack [{shapeType=Box, extentX=0.5, extentY=0.5, extentZ=0.5, center=<-0.0041 -7.681 -8.9284>}]
}
constraints [
 {bone1Name=bat_r, bone2Name=pack, swing1LimitDegrees=5, swing2LimitDegrees=1, twistLimitDegrees=5}
 {bone1Name=bat_l, bone2Name=pack, swing1LimitDegrees=5, swing2LimitDegrees=1, twistLimitDegrees=5}
 {bone1Name=branch1, bone2Name=bat_r, swing1LimitDegrees=3, swing2LimitDegrees=3, twistLimitDegrees=1}
 {bone1Name=branch2, bone2Name=branch1, swing1LimitDegrees=3, swing2LimitDegrees=3, twistLimitDegrees=1}
 {bone1Name=branch3, bone2Name=branch2, swing1LimitDegrees=3, swing2LimitDegrees=3, twistLimitDegrees=1}
 {bone1Name=branchUpr1, bone2Name=bat_r, swing1LimitDegrees=3, swing2LimitDegrees=3, twistLimitDegrees=1}
 {bone1Name=branchUpr2, bone2Name=branchUpr1, swing1LimitDegrees=3, swing2LimitDegrees=3, twistLimitDegrees=1}
]
